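# Translucent unlit Niagara material for the energy tip of the monolith connecting-beam laser, usable on sprites, ribbons and mesh particles.
# A power curve along V scales U around the center, reshaping the UVs that feed a panning fluid-flame texture and the masks.
# The refined texture blends two colors, brightened by a V-based boost and tinted by particle color.
# Opacity multiplies the refined texture by edge masks on both axes and particle alpha.
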Material /PistolCanary_Countdown/Assets/FX/Monoliths/ConnectingBeams/Laser/M_TipEnergy
graph {
 properties {
  bAutomaticallySetUsageInEditor=false
  bUsedWithNiagaraMeshParticles=true
  bUsedWithNiagaraRibbons=true
  bUsedWithNiagaraSprites=true
  blendMode=BLEND_Translucent
  shadingModel=MSM_Unlit
  usedShadingModels=Unlit
 }
 nodes [
  n1 = ParticleColor()
  n2 = TextureObjectParameter() {
   group=MainTexture
   parameterName=MainTexture_Texture
   samplerType=SAMPLERTYPE_LinearGrayscale
   sortPriority=0
   texture=/Game/Weapons/FORT_Melee/Pickaxe_Cube_Ninja_Male/FX/T_FluidFlameTile
  }
  n3 = VectorParameter() {group=MainTexture, parameterName=MainTexture_UVControl, channelNames {a=VPan, b=UPan, g=VScale, r=UScale}, defaultValue=<rgba 0.5 0.07 0 5>}
  n4 = MaterialFunctionCall(Texture=@n2, UVControl_RGB=@n3, UVControl_A=@n3, "[Override]UVInput"=NamedRerouteUsage()) {materialFunction=&FXCommon_TextureSample}
  n5 = VectorParameter() {parameterName=MainTexture_RefineToColor, defaultValue=<rgba 5 1 1 0.5>}
  n6 = Add(
   A=MaterialFunctionCall(Float2=NamedRerouteUsage()) {materialFunction=&BreakOutFloat2Components}.G
   B=ScalarParameter() {defaultValue=-0.2816, group=ColorBoost, parameterName=ColorBoost_GrayAdjust_Add}
  )
  n7 = ScalarParameter() {defaultValue=1.5907, group=ColorBoost, parameterName=ColorBoost_GrayAdjust_Mul}
  n8 = Power(
   Base=Saturate(None=Multiply(A=@n6, B=@n7))
   Exp=ScalarParameter() {defaultValue=3.2513, group=ColorBoost, parameterName=ColorBoost_GrayAdjust_Power}
  )
  n9 = Multiply(
   A=LinearInterpolate(
    A=VectorParameter() {parameterName=Color, defaultValue=<rgba 0.5 0 0.0417 1>}.RGB
    B=VectorParameter() {parameterName=Color2, defaultValue=<rgba 5 0.5 0.875 1>}.RGB
    Alpha=MaterialFunctionCall(Texture=@n4.R, RGB=@n5.RGB, A=@n5.RGB) {materialFunction=&MF_FX_Refine}.Result
   )
   B=LinearInterpolate(B=ScalarParameter() {defaultValue=20, parameterName=B}, Alpha=Saturate(None=@n8)) {constA=1}
  )
  MaterialFunctionCall(Emissive=Divide(A=Multiply(A=@n1.RGB, B=@n9)) {constB=3}) {materialFunction=&MF_FX_EmissiveAdjustments}
  n10 = VectorParameter() {parameterName=MainTexture_Refine, defaultValue=<rgba 0.3 1.4 1 0.5>}
  n11 = MaterialFunctionCall(Float2=NamedRerouteUsage()) {materialFunction=&BreakOutFloat2Components}
  n12 = VectorParameter() {
   group=MainTexture
   parameterName=MainTexture_EdgeMask_InvertControl
   channelNames {a="", b=Power, g=Multiply, r=Offset}
   defaultValue=<rgba 0 2 2 1>
  }
  n13 = VectorParameter() {
   group=MainTexture
   parameterName=MainTexture_EdgeMask_Control
   channelNames {a="", b=Power, g=Multiply, r=Offset}
   defaultValue=<rgba 0 10 1 1>
  }
  n14 = Multiply(
   A=MaterialFunctionCall(Texture=@n4.R, RGB=@n10.RGB, A=@n10.RGB) {materialFunction=&MF_FX_Refine}.Result
   B=MaterialFunctionCall!(UorV=OneMinus(None=@n11.G), Invert_EdgeControl=@n12, NoInvert_EdgeControl=@n13) {materialFunction=/Game/Effects/Niagara/MasterMaterials/Functions/Common/FXCommon_EdgeMask_Ex}.Mask
  )
  n15 = VectorParameter() {group=EdgeFade, parameterName=EdgeFade_EdgeMask_Control, channelNames {a="", b=Power, g=Multiply, r=Offset}, defaultValue=<rgba 0 4 3 1>}
  n16 = MaterialFunctionCall!(UorV=@n11.R, EdgeControl=@n15) {materialFunction=/Game/Effects/Niagara/MasterMaterials/Functions/Common/FXCommon_EdgeMask_Simple}
  Multiply(A=@n1.A, B=Multiply(A=Multiply(A=@n14, B=@n16.Mask), B=ScalarParameter() {defaultValue=2, parameterName=ScaleOp}))
  n17 = OneMinus(None=MaterialFunctionCall(Float2=TextureCoordinate()) {materialFunction=&BreakOutFloat2Components}.G)
  n18 = ScalarParameter() {group=UVScale, parameterName=UVScale_GrayAdjust_Add}
  n19 = Saturate(None=Multiply(A=Add(A=@n17, B=@n18), B=ScalarParameter() {defaultValue=1.8227, group=UVScale, parameterName=UVScale_GrayAdjust_Mul}))
  n20 = ScalarParameter() {defaultValue=0.4343, group=UVScale, parameterName=UVScale_GrayAdjust_Power}
  n21 = OneMinus(None=Saturate(None=Power(Base=@n19, Exp=@n20)))
  n22 = ScalarParameter() {defaultValue=1.6252, group=UVScale, parameterName=UVScale_GrayAdjust_INVPower}
  n23 = LinearInterpolate(Alpha=OneMinus(None=Power(Base=@n21, Exp=@n22)))
  n24 = Constant() {r=1}
  NamedRerouteDeclaration(None=MaterialFunctionCall(
    UVs=TextureCoordinate()
    Texture Scale=MaterialFunctionCall(X=@n23, Y=@n24) {materialFunction=&MakeFloat2}.Result
   ) {materialFunction=&ScaleUVsByCenter}.UVs) {name=ScaledUV, nodeColor=<rgba 0.4766 1 0 1>}
 ]
}
