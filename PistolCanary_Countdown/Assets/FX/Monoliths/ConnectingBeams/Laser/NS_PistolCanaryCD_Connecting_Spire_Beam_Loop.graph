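# Looping Niagara laser beam for the monolith connecting spires, crackling from the owner's position to User.BeamEnd_0.
# Four staggered, re-bursting ribbon emitters (narrow bright white, wide pink glow) lerp particles between endpoints by normalized index, adding randomized waveform offsets that taper at both ends.
# User.BeamFade scales ribbon alpha through per-layer curves, while system age ramps a dynamic material parameter.
# Two GPU spark emitters spray vortex-swirled sprites at each endpoint.
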NiagaraSystem /PistolCanary_Countdown/Assets/FX/Monoliths/ConnectingBeams/Laser/NS_PistolCanaryCD_Connecting_Spire_Beam_Loop
dependencies {
 usedDataInterfaces [&NiagaraDataInterfaceCurve, &NiagaraDataInterfaceCamera]
 usedDynamicInputs [
  &SubtractVectorFromPosition
  /Niagara/DynamicInputs/Helpers/SimulationPosition
  &TransformPosition
  &VectorLength
  &RemapRange
  &ReturnCameraPosition
  &SetBoolByFloatComparison
  &Subtract_Float
  &MakeCustomFloatFromBool
  &DistanceBetweenPositions
 ]
 usedModules [
  &SystemState
  &Timeline
  &InterpolateOverTime
  &self:SystemSpawnScript.SystemScriptSource.SystemScriptGraph.NiagaraNodeAssignment_0.SetVariables_#1
  &self:SystemSpawnScript.SystemScriptSource.SystemScriptGraph.NiagaraNodeAssignment_2.SetVariables_#2
  &self:SystemSpawnScript.SystemScriptSource.SystemScriptGraph.NiagaraNodeAssignment_3.SetVariables_#3
  &self:SystemSpawnScript.SystemScriptSource.SystemScriptGraph.NiagaraNodeAssignment_4.SetVariables_#4
  &self:SystemSpawnScript.SystemScriptSource.SystemScriptGraph.NiagaraNodeAssignment_5.SetVariables_#5
  &self:SystemSpawnScript.SystemScriptSource.SystemScriptGraph.NiagaraNodeAssignment_6.SetVariables_#6
  &self:SystemSpawnScript.SystemScriptSource.SystemScriptGraph.NiagaraNodeAssignment_1.SetVariables_#7
  &self:SystemSpawnScript.SystemScriptSource.SystemScriptGraph.NiagaraNodeAssignment_7.SetVariables_#8
  &self:SystemSpawnScript.SystemScriptSource.SystemScriptGraph.NiagaraNodeAssignment_8.SetVariables_#9
  &self:SystemSpawnScript.SystemScriptSource.SystemScriptGraph.NiagaraNodeAssignment_9.SetVariables_#10
  &self:SystemSpawnScript.SystemScriptSource.SystemScriptGraph.NiagaraNodeAssignment_12.SetVariables_#11
 ]
}
emitters {
 Bright_Ribbon_01 {
  emitterSpawnScript [/SetVariables(
    Emitter.OffsetBeam:NiagaraFloat=&RandomRangeFloat(Minimum=0, Maximum=1)
    Emitter.ParticleLife:NiagaraFloat=&RandomRangeFloat(Minimum=0.5, Maximum=2.25)
    Emitter.ZapSpeed_High=6
    Emitter.ZapSpeed_Low=3
   ) {isSetParameters=true, module=SetVariables_#12}]
  emitterUpdateScript [
   /SetVariables(Emitter.ParticleLife:NiagaraFloat=&RandomRangeFloat(Minimum=0.2, Maximum=0.3)) {isSetParameters=true, module=SetVariables_#13}
   &EmitterState(
    Life Cycle Mode=::NewEnumerator1
    Loop Duration=$Emitter.ParticleLife
    Recalculate Duration Each Loop=true
    Scalability Mode=::NewEnumerator1
    UseLoopDelay=false
   )
   &SpawnBurst_Instantaneous(Spawn Count=12, Use Spawn Probability=false, UseLoopCountLimit=false)
  ]
  particleSpawnScript [
   &InitializeParticle(
    Color=<rgba 7.5 7.5 7.5 1>
    Lifetime=&Add_Float(A=$Emitter.ParticleLife, B=-0.025)
    Position Mode=::NewEnumerator0
    Ribbon Width=&RandomRangeFloat(Minimum=300, Maximum=500)
    Ribbon Width Mode=::NewEnumerator1
    Write Material Random=true
   )
   /SetVariables(
    Particles.DistortionScaleX:NiagaraFloat=&RandomRangeFloat(Minimum=100, Maximum=300)
    Particles.DistortionScaleY:NiagaraFloat=&RandomRangeFloat(Minimum=100, Maximum=300)
    Particles.DistortionScaleZ:NiagaraFloat=&RandomRangeFloat(Minimum=100, Maximum=550)
    Particles.ExecIndex:NiagaraFloat=&ReturnNormalizedExecIndex()
    Particles.ZapSpeed_High:NiagaraFloat=$Emitter.ZapSpeed_High
    Particles.ZapSpeed_Low:NiagaraFloat=$Emitter.ZapSpeed_Low
   ) {isSetParameters=true, module=SetVariables_#14}
   &DynamicMaterialParameters(
    Index 0 Param 1=0.75
    Index 0 Param 2=0.02
    Index 0 Param 3=$Emitter.OffsetBeam
    Index 0 Param 4=1
    Module.Index0Float0Write=true
    Module.Index0Float1Write=true
    Module.Index0Float2Write=true
    Module.Index0Float3Write=true
   )
  ]
  particleUpdateScript [
   &SolveForcesAndVelocity(Clamp Velocity=false, Limit Acceleration=false)
   &ParticleState()
   /SetVariables(Particles.Position=&AddVectorToPosition(
     Position=&Lerp_Vector(
      A=Linked() {linkedVariable {name=Engine.Owner.Position, type=NiagaraPosition}}
      B=Linked() {linkedVariable {name=User.BeamEnd_0, type=NiagaraPosition}}
      Alpha=$Particles.ExecIndex
     )
     Vector=&Multiply_VectorByFloat(
      Vector=&MakeVector(
       X=&Waveform(
        Waveform Count=::NewEnumerator1
        Global Phase=&Multiply_Float(A=$Emitter.Age, B=&RandomRangeFloat(Minimum=$Particles.ZapSpeed_Low, Maximum=$Particles.ZapSpeed_High))
        "Waveform [1]"=::NewEnumerator0
        "[1] Frequency"=&RandomRangeFloat(Minimum=2, Maximum=4)
        "[1] Amplitude Min/Max"=<0 1>
        "[1] Scale"=$Particles.DistortionScaleX
        "[2] Average"=false
        "[2] Input Position"=0
        "[2] Phase"=$Emitter.Age
        "[2] Phase Scale"=1
        "[2] Compound Frequency"=2
        "[2] Frequency"=2
        "[2] Amplitude Min/Max"=<0 1>
        "[2] Scale"=$Particles.DistortionScaleX
        "[2] Bias"=0
        "[2] Exponent"=1
        "[2] Pulse Duration"=0.1
        "[2] Clip Triangle Min/Max"=<0 0>
        "[2] Random Seed"=0
       )
       Y=&Waveform(
        Waveform Count=::NewEnumerator2
        Global Phase=&Multiply_Float(A=$Emitter.Age, B=&RandomRangeFloat(Minimum=$Particles.ZapSpeed_Low, Maximum=$Particles.ZapSpeed_High))
        "Waveform [1]"=::NewEnumerator0
        "[1] Frequency"=1
        "[1] Amplitude Min/Max"=<0 1>
        "[1] Scale"=$Particles.DistortionScaleY
        "[2] Average"=false
        "[2] Input Position"=0
        "[2] Phase"=$Emitter.Age
        "[2] Phase Scale"=1
        "[2] Compound Frequency"=2
        "[2] Frequency"=2
        "[2] Amplitude Min/Max"=<0 1>
        "[2] Scale"=$Particles.DistortionScaleX
        "[2] Bias"=0
        "[2] Exponent"=1
        "[2] Pulse Duration"=0.1
        "[2] Clip Triangle Min/Max"=<0 0>
        "[2] Random Seed"=0
        "[3] Average"=false
        "Waveform [3]"=::NewEnumerator1
        "[3] Input Position"=0
        "[3] Phase"=$Emitter.Age
        "[3] Phase Scale"=1
        "[3] Frequency"=4
        "[3] Compound Frequency"=2
        "[3] Amplitude Min/Max"=<0 1>
        "[3] Scale"=1
        "[3] Bias"=0
        "[3] Exponent"=1
        "[3] Pulse Duration"=0.1
        "[3] Clip Triangle Min/Max"=<0 0>
        "[3] Random Seed"=0
       )
       Z=&Waveform(
        Waveform Count=::NewEnumerator1
        Global Phase=&Multiply_Float(A=$Emitter.Age, B=&RandomRangeFloat(Minimum=$Particles.ZapSpeed_Low, Maximum=$Particles.ZapSpeed_High))
        "Waveform [1]"=::NewEnumerator0
        "[1] Frequency"=1
        "[1] Amplitude Min/Max"=<0 1>
        "[1] Scale"=$Particles.DistortionScaleZ
        "[2] Average"=false
        "[2] Input Position"=0
        "[2] Phase"=$Emitter.Age
        "[2] Phase Scale"=1
        "[2] Compound Frequency"=2
        "[2] Frequency"=&RandomRangeFloat(Minimum=0, Maximum=2)
        "[2] Amplitude Min/Max"=<-0.1 0.1>
        "[2] Scale"=$Particles.DistortionScaleZ
        "[2] Bias"=0
        "[2] Exponent"=2
        "[2] Pulse Duration"=0.1
        "[2] Clip Triangle Min/Max"=<0 0>
        "[2] Random Seed"=0
        "Waveform [3]"=::NewEnumerator1
       )
      )
      Float=&FloatFromCurve(
       FloatCurve=DataInterface() {propertyValues {Curve {
          defaultValue=1
          keys [
           {arriveTangent=1.8269, interpMode=RCIM_Cubic, leaveTangent=1.8269, tangentMode=RCTM_User, time=0, value=0}
           {interpMode=RCIM_Cubic, tangentMode=RCTM_User, time=0.5, value=1}
           {arriveTangent=-1.1725, tangentMode=RCTM_User, time=1, value=0}
          ]
         }}}
       CurveIndex=$Particles.ExecIndex
      )
     )
    )) {isSetParameters=true, module=SetVariables_#15}
   &DynamicMaterialParameters(
    Index 0 Param 4=&FloatFromCurve(
     FloatCurve=DataInterface() {propertyValues {Curve {
        defaultValue=1
        keys [{interpMode=RCIM_Cubic, tangentMode=RCTM_User, time=0, value=0}, {interpMode=RCIM_Cubic, tangentMode=RCTM_User, time=1, value=1}]
       }}}
     CurveIndex=&ClampFloat(Float=&Divide_Float(A=$System.Age, B=5))
    )
    Module.Index0Float0Write=false
    Module.Index0Float1Write=false
    Module.Index0Float2Write=false
    Module.Index0Float3Write=true
   ) {module=DynamicMaterialParameters001}
   &ScaleColor(
    Scale Alpha=&FloatFromCurve(
     FloatCurve=DataInterface() {propertyValues {Curve {keys [<at 0 0 RCIM_Cubic>, {interpMode=RCIM_Cubic, tangentMode=RCTM_User, time=0.05, value=1}, <at 1 1>]}}}
     CurveIndex=$User.BeamFade
     Scale Curve=&FloatFromCurve(
      FloatCurve=DataInterface() {propertyValues {Curve {
         defaultValue=1
         keys [{interpMode=RCIM_Cubic, tangentMode=RCTM_User, time=0, value=0}, {interpMode=RCIM_Cubic, tangentMode=RCTM_User, time=1, value=1}]
        }}}
      CurveIndex=$User.BeamFade
     )
    )
    ScaleA=true
    ScaleRGB=true
   )
  ]
  rendererClasses [&NiagaraRibbonRendererProperties]
  renderers [{rendererClass=&NiagaraRibbonRendererProperties}]
 }
 Bright_Ribbon_02 {
  emitterSpawnScript [/SetVariables(
    Emitter.OffsetBeam:NiagaraFloat=&RandomRangeFloat(Minimum=0, Maximum=1)
    Emitter.ParticleLife:NiagaraFloat=&RandomRangeFloat(Minimum=0.5, Maximum=2.25)
    Emitter.ZapSpeed_High=6
    Emitter.ZapSpeed_Low=3
   ) {isSetParameters=true, module=SetVariables_#12}]
  emitterUpdateScript [
   /SetVariables(Emitter.ParticleLife:NiagaraFloat=&RandomRangeFloat(Minimum=0.2, Maximum=0.5)) {isSetParameters=true, module=SetVariables_#13}
   &EmitterState(
    Delay First Loop Only=true
    Life Cycle Mode=::NewEnumerator1
    Loop Delay=0.1
    Loop Duration=$Emitter.ParticleLife
    Recalculate Duration Each Loop=true
    Scalability Mode=::NewEnumerator1
    UseLoopDelay=true
   )
   &SpawnBurst_Instantaneous(Spawn Count=13, Use Spawn Probability=false, UseLoopCountLimit=false)
  ]
  particleSpawnScript [
   &InitializeParticle(
    Color=<rgba 7.5 7.5 7.5 1>
    Lifetime=&Add_Float(A=$Emitter.ParticleLife, B=-0.025)
    Position Mode=::NewEnumerator0
    Ribbon Width=&RandomRangeFloat(Minimum=300, Maximum=500)
    Ribbon Width Mode=::NewEnumerator1
    Write Material Random=true
   )
   /SetVariables(
    Particles.DistortionScaleX:NiagaraFloat=&RandomRangeFloat(Minimum=100, Maximum=300)
    Particles.DistortionScaleY:NiagaraFloat=&RandomRangeFloat(Minimum=100, Maximum=300)
    Particles.DistortionScaleZ:NiagaraFloat=&RandomRangeFloat(Minimum=100, Maximum=550)
    Particles.ExecIndex:NiagaraFloat=&ReturnNormalizedExecIndex()
    Particles.ZapSpeed_High:NiagaraFloat=$Emitter.ZapSpeed_High
    Particles.ZapSpeed_Low:NiagaraFloat=$Emitter.ZapSpeed_Low
   ) {isSetParameters=true}
   &DynamicMaterialParameters(
    Index 0 Param 1=1
    Index 0 Param 2=0.02
    Index 0 Param 3=$Emitter.OffsetBeam
    Index 0 Param 4=1
    Module.Index0Float0Write=true
    Module.Index0Float1Write=true
    Module.Index0Float2Write=true
    Module.Index0Float3Write=true
   )
  ]
  particleUpdateScript [
   &SolveForcesAndVelocity(Clamp Velocity=false, Limit Acceleration=false)
   &ParticleState()
   /SetVariables(Particles.Position=&AddVectorToPosition(
     Position=&Lerp_Vector(
      A=Linked() {linkedVariable {name=Engine.Owner.Position, type=NiagaraPosition}}
      B=Linked() {linkedVariable {name=User.BeamEnd_0, type=NiagaraPosition}}
      Alpha=$Particles.ExecIndex
     )
     Vector=&Multiply_VectorByFloat(
      Vector=&MakeVector(
       X=&Waveform(
        Waveform Count=::NewEnumerator1
        Global Phase=&Multiply_Float(A=$Emitter.Age, B=&RandomRangeFloat(Minimum=$Particles.ZapSpeed_Low, Maximum=$Particles.ZapSpeed_High))
        "Waveform [1]"=::NewEnumerator0
        "[1] Frequency"=&RandomRangeFloat(Minimum=2, Maximum=4)
        "[1] Amplitude Min/Max"=<0 1>
        "[1] Scale"=$Particles.DistortionScaleX
        "[2] Average"=false
        "[2] Input Position"=0
        "[2] Phase"=$Emitter.Age
        "[2] Phase Scale"=1
        "[2] Compound Frequency"=2
        "[2] Frequency"=2
        "[2] Amplitude Min/Max"=<0 1>
        "[2] Scale"=$Particles.DistortionScaleX
        "[2] Bias"=0
        "[2] Exponent"=1
        "[2] Pulse Duration"=0.1
        "[2] Clip Triangle Min/Max"=<0 0>
        "[2] Random Seed"=0
       )
       Y=&Waveform(
        Waveform Count=::NewEnumerator2
        Global Phase=&Multiply_Float(A=$Emitter.Age, B=&RandomRangeFloat(Minimum=$Particles.ZapSpeed_Low, Maximum=$Particles.ZapSpeed_High))
        "Waveform [1]"=::NewEnumerator0
        "[1] Frequency"=1
        "[1] Amplitude Min/Max"=<0 1>
        "[1] Scale"=$Particles.DistortionScaleY
        "[2] Average"=false
        "[2] Input Position"=0
        "[2] Phase"=$Emitter.Age
        "[2] Phase Scale"=1
        "[2] Compound Frequency"=2
        "[2] Frequency"=2
        "[2] Amplitude Min/Max"=<0 1>
        "[2] Scale"=$Particles.DistortionScaleX
        "[2] Bias"=0
        "[2] Exponent"=1
        "[2] Pulse Duration"=0.1
        "[2] Clip Triangle Min/Max"=<0 0>
        "[2] Random Seed"=0
        "[3] Average"=false
        "Waveform [3]"=::NewEnumerator1
        "[3] Input Position"=0
        "[3] Phase"=$Emitter.Age
        "[3] Phase Scale"=1
        "[3] Frequency"=4
        "[3] Compound Frequency"=2
        "[3] Amplitude Min/Max"=<0 1>
        "[3] Scale"=1
        "[3] Bias"=0
        "[3] Exponent"=1
        "[3] Pulse Duration"=0.1
        "[3] Clip Triangle Min/Max"=<0 0>
        "[3] Random Seed"=0
       )
       Z=&Waveform(
        Waveform Count=::NewEnumerator1
        Global Phase=&Multiply_Float(A=$Emitter.Age, B=&RandomRangeFloat(Minimum=$Particles.ZapSpeed_Low, Maximum=$Particles.ZapSpeed_High))
        "Waveform [1]"=::NewEnumerator0
        "[1] Frequency"=1
        "[1] Amplitude Min/Max"=<0 1>
        "[1] Scale"=$Particles.DistortionScaleZ
        "[2] Average"=false
        "[2] Input Position"=0
        "[2] Phase"=$Emitter.Age
        "[2] Phase Scale"=1
        "[2] Compound Frequency"=2
        "[2] Frequency"=&RandomRangeFloat(Minimum=0, Maximum=2)
        "[2] Amplitude Min/Max"=<-0.1 0.1>
        "[2] Scale"=$Particles.DistortionScaleZ
        "[2] Bias"=0
        "[2] Exponent"=2
        "[2] Pulse Duration"=0.1
        "[2] Clip Triangle Min/Max"=<0 0>
        "[2] Random Seed"=0
        "Waveform [3]"=::NewEnumerator1
       )
      )
      Float=&FloatFromCurve(
       FloatCurve=DataInterface() {propertyValues {Curve {
          defaultValue=1
          keys [
           {arriveTangent=1.8269, interpMode=RCIM_Cubic, leaveTangent=1.8269, tangentMode=RCTM_User, time=0, value=0}
           {interpMode=RCIM_Cubic, tangentMode=RCTM_User, time=0.5, value=1}
           {arriveTangent=-1.1725, tangentMode=RCTM_User, time=1, value=0}
          ]
         }}}
       CurveIndex=$Particles.ExecIndex
      )
     )
    )) {isSetParameters=true, module=SetVariables_#15}
   &DynamicMaterialParameters(
    Index 0 Param 4=&FloatFromCurve(
     FloatCurve=DataInterface() {propertyValues {Curve {
        defaultValue=1
        keys [{interpMode=RCIM_Cubic, tangentMode=RCTM_User, time=0, value=0}, {interpMode=RCIM_Cubic, tangentMode=RCTM_User, time=1, value=1}]
       }}}
     CurveIndex=&ClampFloat(Float=&Divide_Float(A=$System.Age, B=2))
    )
    Module.Index0Float0Write=false
    Module.Index0Float1Write=false
    Module.Index0Float2Write=false
    Module.Index0Float3Write=true
   ) {module=DynamicMaterialParameters001}
   &ScaleColor(
    Scale Alpha=&FloatFromCurve(
     FloatCurve=DataInterface() {propertyValues {Curve {keys [
         <at 0 0 RCIM_Cubic>
         {arriveTangent=4, interpMode=RCIM_Cubic, leaveTangent=4, time=0.2, value=0.2}
         {interpMode=RCIM_Cubic, tangentMode=RCTM_User, time=0.25, value=1}
         <at 1 1>
        ]}}}
     CurveIndex=$User.BeamFade
     Scale Curve=&FloatFromCurve(
      FloatCurve=DataInterface() {propertyValues {Curve {
         defaultValue=1
         keys [{interpMode=RCIM_Cubic, tangentMode=RCTM_User, time=0, value=0}, {interpMode=RCIM_Cubic, tangentMode=RCTM_User, time=1, value=1}]
        }}}
      CurveIndex=$User.BeamFade
     )
    )
    ScaleA=true
    ScaleRGB=true
   )
  ]
  rendererClasses [&NiagaraRibbonRendererProperties]
  renderers [{rendererClass=&NiagaraRibbonRendererProperties}]
 }
 Faded_Ribbon_01 {
  emitterSpawnScript [/SetVariables(
    Emitter.OffsetBeam:NiagaraFloat=&RandomRangeFloat(Minimum=0, Maximum=1)
    Emitter.ParticleLife:NiagaraFloat=&RandomRangeFloat(Minimum=0.5, Maximum=0.75)
    Emitter.ZapSpeed_High=3
    Emitter.ZapSpeed_Low=1
   ) {isSetParameters=true, module=SetVariables_#12}]
  emitterUpdateScript [
   &EmitterState(
    Inactive Response=::NewEnumerator1
    Life Cycle Mode=::NewEnumerator1
    Loop Duration=$Emitter.Initial.ParticleLife
    Recalculate Duration Each Loop=true
    Scalability Mode=::NewEnumerator1
    UseLoopDelay=false
   )
   &SpawnBurst_Instantaneous(Spawn Count=10, Use Spawn Probability=false, UseLoopCountLimit=false)
  ]
  particleSpawnScript [
   &InitializeParticle(
    Color=<rgba 5 0.25 1.6596 1>
    Lifetime=$Emitter.ParticleLife
    Position Mode=::NewEnumerator0
    Ribbon Width=&RandomRangeFloat(Minimum=750, Maximum=1200)
    Ribbon Width Mode=::NewEnumerator1
    Write Material Random=true
   )
   /SetVariables(
    Particles.DistortionScaleX:NiagaraFloat=&RandomRangeFloat(Minimum=100, Maximum=750)
    Particles.DistortionScaleY:NiagaraFloat=&RandomRangeFloat(Minimum=100, Maximum=750)
    Particles.DistortionScaleZ:NiagaraFloat=&RandomRangeFloat(Minimum=100, Maximum=1200)
    Particles.ExecIndex:NiagaraFloat=&ReturnNormalizedExecIndex()
    Particles.ZapSpeed_High:NiagaraFloat=$Emitter.ZapSpeed_High
    Particles.ZapSpeed_Low:NiagaraFloat=$Emitter.ZapSpeed_Low
   ) {isSetParameters=true, module=SetVariables_#14}
   &DynamicMaterialParameters(
    Index 0 Param 1=1
    Index 0 Param 2=0.02
    Index 0 Param 3=$Emitter.OffsetBeam
    Index 0 Param 4=1
    Module.Index0Float0Write=true
    Module.Index0Float1Write=true
    Module.Index0Float2Write=true
    Module.Index0Float3Write=true
   )
  ]
  particleUpdateScript [
   &SolveForcesAndVelocity(Clamp Velocity=false, Limit Acceleration=false)
   &ParticleState()
   &ScaleColor(
    Scale Alpha=&FloatFromCurve(
     FloatCurve=DataInterface() {propertyValues {Curve {
        defaultValue=0
        keys [
         {interpMode=RCIM_Cubic, tangentMode=RCTM_User, time=0, value=0}
         {interpMode=RCIM_Cubic, tangentMode=RCTM_User, time=0.5, value=1}
         {tangentMode=RCTM_User, time=1, value=0}
        ]
       }}}
     Scale Curve=&FloatFromCurve(
      FloatCurve=DataInterface() {propertyValues {Curve {keys [
          <at 0 0 RCIM_Constant>
          {interpMode=RCIM_Cubic, tangentMode=RCTM_User, time=0.25, value=0}
          {arriveTangent=4, interpMode=RCIM_Cubic, leaveTangent=4, time=0.4, value=0.25}
          <at 0.5 1>
          <at 1 1>
         ]}}}
      CurveIndex=$User.BeamFade
      Scale Curve=&FloatFromCurve(
       FloatCurve=DataInterface() {propertyValues {Curve {
          defaultValue=1
          keys [{interpMode=RCIM_Cubic, tangentMode=RCTM_User, time=0, value=0}, {interpMode=RCIM_Cubic, tangentMode=RCTM_User, time=1, value=1}]
         }}}
       CurveIndex=$User.BeamFade
      )
     )
    )
    Scale RGB=&VectorFromFloat(Value=&Add_Float(
      A=&FloatFromCurve(
       FloatCurve=DataInterface() {propertyValues {Curve {keys [<at 0 1 RCIM_Cubic>,<at 0.5 0 RCIM_Cubic>,<at 1 1>]}}}
       CurveIndex=$Particles.ExecIndex
       Scale Curve=2
      )
      B=0.75
     ))
    ScaleA=true
    ScaleRGB=true
   )
   /SetVariables(Particles.Position=&AddVectorToPosition(
     Position=&Lerp_Vector(
      A=Linked() {linkedVariable {name=Engine.Owner.Position, type=NiagaraPosition}}
      B=Linked() {linkedVariable {name=User.BeamEnd_0, type=NiagaraPosition}}
      Alpha=$Particles.ExecIndex
     )
     Vector=&Multiply_VectorByFloat(
      Vector=&MakeVector(
       X=&Waveform(
        Waveform Count=::NewEnumerator1
        Global Phase=&Multiply_Float(A=$Emitter.Age, B=&RandomRangeFloat(Minimum=$Particles.ZapSpeed_Low, Maximum=$Particles.ZapSpeed_High))
        "Waveform [1]"=::NewEnumerator0
        "[1] Frequency"=&RandomRangeFloat(Minimum=2, Maximum=4)
        "[1] Amplitude Min/Max"=<0 1>
        "[1] Scale"=$Particles.DistortionScaleX
        "[2] Average"=false
        "[2] Input Position"=0
        "[2] Phase"=$Emitter.Age
        "[2] Phase Scale"=1
        "[2] Compound Frequency"=2
        "[2] Frequency"=2
        "[2] Amplitude Min/Max"=<0 1>
        "[2] Scale"=$Particles.DistortionScaleX
        "[2] Bias"=0
        "[2] Exponent"=1
        "[2] Pulse Duration"=0.1
        "[2] Clip Triangle Min/Max"=<0 0>
        "[2] Random Seed"=0
       )
       Y=&Waveform(
        Waveform Count=::NewEnumerator2
        Global Phase=&Multiply_Float(A=$Emitter.Age, B=&RandomRangeFloat(Minimum=$Particles.ZapSpeed_Low, Maximum=$Particles.ZapSpeed_High))
        "Waveform [1]"=::NewEnumerator0
        "[1] Frequency"=1
        "[1] Amplitude Min/Max"=<0 1>
        "[1] Scale"=$Particles.DistortionScaleY
        "[2] Average"=false
        "[2] Input Position"=0
        "[2] Phase"=$Emitter.Age
        "[2] Phase Scale"=1
        "[2] Compound Frequency"=2
        "[2] Frequency"=2
        "[2] Amplitude Min/Max"=<0 1>
        "[2] Scale"=$Particles.DistortionScaleX
        "[2] Bias"=0
        "[2] Exponent"=1
        "[2] Pulse Duration"=0.1
        "[2] Clip Triangle Min/Max"=<0 0>
        "[2] Random Seed"=0
        "[3] Average"=false
        "Waveform [3]"=::NewEnumerator1
        "[3] Input Position"=0
        "[3] Phase"=$Emitter.Age
        "[3] Phase Scale"=1
        "[3] Frequency"=4
        "[3] Compound Frequency"=2
        "[3] Amplitude Min/Max"=<0 1>
        "[3] Scale"=1
        "[3] Bias"=0
        "[3] Exponent"=1
        "[3] Pulse Duration"=0.1
        "[3] Clip Triangle Min/Max"=<0 0>
        "[3] Random Seed"=0
       )
       Z=&Waveform(
        Waveform Count=::NewEnumerator1
        Global Phase=&Multiply_Float(A=$Emitter.Age, B=&RandomRangeFloat(Minimum=$Particles.ZapSpeed_Low, Maximum=$Particles.ZapSpeed_High))
        "Waveform [1]"=::NewEnumerator0
        "[1] Frequency"=1
        "[1] Amplitude Min/Max"=<0 1>
        "[1] Scale"=$Particles.DistortionScaleZ
        "[2] Average"=false
        "[2] Input Position"=0
        "[2] Phase"=$Emitter.Age
        "[2] Phase Scale"=1
        "[2] Compound Frequency"=2
        "[2] Frequency"=&RandomRangeFloat(Minimum=0, Maximum=2)
        "[2] Amplitude Min/Max"=<-0.1 0.1>
        "[2] Scale"=$Particles.DistortionScaleZ
        "[2] Bias"=0
        "[2] Exponent"=2
        "[2] Pulse Duration"=0.1
        "[2] Clip Triangle Min/Max"=<0 0>
        "[2] Random Seed"=0
        "Waveform [3]"=::NewEnumerator1
       )
      )
      Float=&FloatFromCurve(
       FloatCurve=DataInterface() {propertyValues {Curve {
          defaultValue=1
          keys [
           {arriveTangent=1.1786, interpMode=RCIM_Cubic, leaveTangent=1.1786, tangentMode=RCTM_User, time=0, value=0}
           {interpMode=RCIM_Cubic, tangentMode=RCTM_User, time=0.5, value=1}
           {arriveTangent=-0.99, tangentMode=RCTM_User, time=1, value=0}
          ]
         }}}
       CurveIndex=$Particles.ExecIndex
      )
     )
    )) {isSetParameters=true, module=SetVariables_#15}
   &DynamicMaterialParameters(
    Index 0 Param 4=&FloatFromCurve(
     FloatCurve=DataInterface() {propertyValues {Curve {
        defaultValue=1
        keys [{interpMode=RCIM_Cubic, tangentMode=RCTM_User, time=0, value=0}, {interpMode=RCIM_Cubic, tangentMode=RCTM_User, time=1, value=1}]
       }}}
     CurveIndex=&ClampFloat(Float=&Divide_Float(A=$System.Age, B=2))
    )
    Module.Index0Float0Write=false
    Module.Index0Float1Write=false
    Module.Index0Float2Write=false
    Module.Index0Float3Write=true
   ) {module=DynamicMaterialParameters001}
  ]
  rendererClasses [&NiagaraRibbonRendererProperties]
  renderers [{rendererClass=&NiagaraRibbonRendererProperties}]
 }
 Faded_Ribbon_02 {
  emitterSpawnScript [/SetVariables(
    Emitter.OffsetBeam:NiagaraFloat=&RandomRangeFloat(Minimum=0, Maximum=1)
    Emitter.ParticleLife:NiagaraFloat=&RandomRangeFloat(Minimum=0.5, Maximum=0.75)
    Emitter.ZapSpeed_High=3
    Emitter.ZapSpeed_Low=1
   ) {isSetParameters=true, module=SetVariables_#12}]
  emitterUpdateScript [
   &EmitterState(
    Delay First Loop Only=true
    Inactive Response=::NewEnumerator1
    Life Cycle Mode=::NewEnumerator1
    Loop Delay=0.275
    Loop Duration=$Emitter.Initial.ParticleLife
    Recalculate Duration Each Loop=true
    Scalability Mode=::NewEnumerator1
    UseLoopDelay=true
   )
   &SpawnBurst_Instantaneous(Spawn Count=7, Use Spawn Probability=false, UseLoopCountLimit=false)
  ]
  particleSpawnScript [
   &InitializeParticle(
    Color=<rgba 5 0.25 1.6596 1>
    Lifetime=$Emitter.ParticleLife
    Position Mode=::NewEnumerator0
    Ribbon Width=&RandomRangeFloat(Minimum=750, Maximum=1200)
    Ribbon Width Mode=::NewEnumerator1
    Write Material Random=true
   )
   /SetVariables(
    Particles.DistortionScaleX:NiagaraFloat=&RandomRangeFloat(Minimum=100, Maximum=750)
    Particles.DistortionScaleY:NiagaraFloat=&RandomRangeFloat(Minimum=100, Maximum=750)
    Particles.DistortionScaleZ:NiagaraFloat=&RandomRangeFloat(Minimum=100, Maximum=1200)
    Particles.ExecIndex:NiagaraFloat=&ReturnNormalizedExecIndex()
    Particles.ZapSpeed_High:NiagaraFloat=$Emitter.ZapSpeed_High
    Particles.ZapSpeed_Low:NiagaraFloat=$Emitter.ZapSpeed_Low
   ) {isSetParameters=true, module=SetVariables_#14}
   &DynamicMaterialParameters(
    Index 0 Param 1=1
    Index 0 Param 2=0.02
    Index 0 Param 3=$Emitter.OffsetBeam
    Index 0 Param 4=1
    Module.Index0Float0Write=true
    Module.Index0Float1Write=true
    Module.Index0Float2Write=true
    Module.Index0Float3Write=true
   )
  ]
  particleUpdateScript [
   &SolveForcesAndVelocity(Clamp Velocity=false, Limit Acceleration=false)
   &ParticleState()
   &ScaleColor(
    Scale Alpha=&FloatFromCurve(
     FloatCurve=DataInterface() {propertyValues {Curve {
        defaultValue=0
        keys [
         {interpMode=RCIM_Cubic, tangentMode=RCTM_User, time=0, value=0}
         {interpMode=RCIM_Cubic, tangentMode=RCTM_User, time=0.5, value=1}
         {tangentMode=RCTM_User, time=1, value=0}
        ]
       }}}
     Scale Curve=&FloatFromCurve(
      FloatCurve=DataInterface() {propertyValues {Curve {keys [
          <at 0 0 RCIM_Constant>
          {interpMode=RCIM_Cubic, tangentMode=RCTM_User, time=0.5, value=0}
          {arriveTangent=4, interpMode=RCIM_Cubic, leaveTangent=4, time=0.7, value=0.3}
          <at 0.75 1>
          <at 1 1>
         ]}}}
      CurveIndex=$User.BeamFade
      Scale Curve=&FloatFromCurve(
       FloatCurve=DataInterface() {propertyValues {Curve {
          defaultValue=1
          keys [{interpMode=RCIM_Cubic, tangentMode=RCTM_User, time=0, value=0}, {interpMode=RCIM_Cubic, tangentMode=RCTM_User, time=1, value=1}]
         }}}
       CurveIndex=$User.BeamFade
      )
     )
    )
    Scale RGB=&VectorFromFloat(Value=&Add_Float(
      A=&FloatFromCurve(
       FloatCurve=DataInterface() {propertyValues {Curve {keys [<at 0 1 RCIM_Cubic>,<at 0.5 0 RCIM_Cubic>,<at 1 1>]}}}
       CurveIndex=$Particles.ExecIndex
       Scale Curve=2
      )
      B=0.75
     ))
    ScaleA=true
    ScaleRGB=true
   )
   /SetVariables(Particles.Position=&AddVectorToPosition(
     Position=&Lerp_Vector(
      A=Linked() {linkedVariable {name=Engine.Owner.Position, type=NiagaraPosition}}
      B=Linked() {linkedVariable {name=User.BeamEnd_0, type=NiagaraPosition}}
      Alpha=$Particles.ExecIndex
     )
     Vector=&Multiply_VectorByFloat(
      Vector=&MakeVector(
       X=&Waveform(
        Waveform Count=::NewEnumerator1
        Global Phase=&Multiply_Float(A=$Emitter.Age, B=&RandomRangeFloat(Minimum=$Particles.ZapSpeed_Low, Maximum=$Particles.ZapSpeed_High))
        "Waveform [1]"=::NewEnumerator0
        "[1] Frequency"=&RandomRangeFloat(Minimum=2, Maximum=4)
        "[1] Amplitude Min/Max"=<0 1>
        "[1] Scale"=$Particles.DistortionScaleX
        "[2] Average"=false
        "[2] Input Position"=0
        "[2] Phase"=$Emitter.Age
        "[2] Phase Scale"=1
        "[2] Compound Frequency"=2
        "[2] Frequency"=2
        "[2] Amplitude Min/Max"=<0 1>
        "[2] Scale"=$Particles.DistortionScaleX
        "[2] Bias"=0
        "[2] Exponent"=1
        "[2] Pulse Duration"=0.1
        "[2] Clip Triangle Min/Max"=<0 0>
        "[2] Random Seed"=0
       )
       Y=&Waveform(
        Waveform Count=::NewEnumerator2
        Global Phase=&Multiply_Float(A=$Emitter.Age, B=&RandomRangeFloat(Minimum=$Particles.ZapSpeed_Low, Maximum=$Particles.ZapSpeed_High))
        "Waveform [1]"=::NewEnumerator0
        "[1] Frequency"=1
        "[1] Amplitude Min/Max"=<0 1>
        "[1] Scale"=$Particles.DistortionScaleY
        "[2] Average"=false
        "[2] Input Position"=0
        "[2] Phase"=$Emitter.Age
        "[2] Phase Scale"=1
        "[2] Compound Frequency"=2
        "[2] Frequency"=2
        "[2] Amplitude Min/Max"=<0 1>
        "[2] Scale"=$Particles.DistortionScaleX
        "[2] Bias"=0
        "[2] Exponent"=1
        "[2] Pulse Duration"=0.1
        "[2] Clip Triangle Min/Max"=<0 0>
        "[2] Random Seed"=0
        "[3] Average"=false
        "Waveform [3]"=::NewEnumerator1
        "[3] Input Position"=0
        "[3] Phase"=$Emitter.Age
        "[3] Phase Scale"=1
        "[3] Frequency"=4
        "[3] Compound Frequency"=2
        "[3] Amplitude Min/Max"=<0 1>
        "[3] Scale"=1
        "[3] Bias"=0
        "[3] Exponent"=1
        "[3] Pulse Duration"=0.1
        "[3] Clip Triangle Min/Max"=<0 0>
        "[3] Random Seed"=0
       )
       Z=&Waveform(
        Waveform Count=::NewEnumerator1
        Global Phase=&Multiply_Float(A=$Emitter.Age, B=&RandomRangeFloat(Minimum=$Particles.ZapSpeed_Low, Maximum=$Particles.ZapSpeed_High))
        "Waveform [1]"=::NewEnumerator0
        "[1] Frequency"=1
        "[1] Amplitude Min/Max"=<0 1>
        "[1] Scale"=$Particles.DistortionScaleZ
        "[2] Average"=false
        "[2] Input Position"=0
        "[2] Phase"=$Emitter.Age
        "[2] Phase Scale"=1
        "[2] Compound Frequency"=2
        "[2] Frequency"=&RandomRangeFloat(Minimum=0, Maximum=2)
        "[2] Amplitude Min/Max"=<-0.1 0.1>
        "[2] Scale"=$Particles.DistortionScaleZ
        "[2] Bias"=0
        "[2] Exponent"=2
        "[2] Pulse Duration"=0.1
        "[2] Clip Triangle Min/Max"=<0 0>
        "[2] Random Seed"=0
        "Waveform [3]"=::NewEnumerator1
       )
      )
      Float=&FloatFromCurve(
       FloatCurve=DataInterface() {propertyValues {Curve {
          defaultValue=1
          keys [
           {arriveTangent=1.1786, interpMode=RCIM_Cubic, leaveTangent=1.1786, tangentMode=RCTM_User, time=0, value=0}
           {interpMode=RCIM_Cubic, tangentMode=RCTM_User, time=0.5, value=1}
           {arriveTangent=-0.99, tangentMode=RCTM_User, time=1, value=0}
          ]
         }}}
       CurveIndex=$Particles.ExecIndex
      )
     )
    )) {isSetParameters=true, module=SetVariables_#15}
   &DynamicMaterialParameters(
    Index 0 Param 4=&FloatFromCurve(
     FloatCurve=DataInterface() {propertyValues {Curve {
        defaultValue=1
        keys [{interpMode=RCIM_Cubic, tangentMode=RCTM_User, time=0, value=0}, {interpMode=RCIM_Cubic, tangentMode=RCTM_User, time=1, value=1}]
       }}}
     CurveIndex=&ClampFloat(Float=&Divide_Float(A=$System.Age, B=2))
    )
    Module.Index0Float0Write=false
    Module.Index0Float1Write=false
    Module.Index0Float2Write=false
    Module.Index0Float3Write=true
   ) {module=DynamicMaterialParameters001}
  ]
  rendererClasses [&NiagaraRibbonRendererProperties]
  renderers [{rendererClass=&NiagaraRibbonRendererProperties}]
 }
 Sparks_Main {
  simTarget=GPUComputeSim
  emitterUpdateScript [
   &EmitterState(Life Cycle Mode=::NewEnumerator1, Loop Duration=1, UseLoopDelay=false)
   &SpawnRate(Spawn Probability=0.2, SpawnRate=100, Use Spawn Probability=true)
  ]
  particleSpawnScript [
   &InitializeParticle(
    Color=<rgba 275 65.3025 230.0222 1>
    Lifetime Min=0.5
    Lifetime Mode=::NewEnumerator1
    Position Offset=<0 0 100>
    Sprite Size=&Multiply_Vector2D_ByFloat(Float=&RandomRangeFloat(Minimum=0.1, Maximum=1), Vector2D=<40 65>)
    Sprite Size Mode=::NewEnumerator4
    UsePositionOffset=true
    Write Material Random=true
   )
   &ShapeLocation(MaskSpawnGroup=false, Override Seed=false, Sphere Radius=50, UseOwnerScale=true)
   &AddVelocity(
    Cone Angle=65
    Module.UseConeFalloff=false
    Module.UseVelDistribution=false
    Override Seed=false
    Velocity Mode=::NewEnumerator2
    Velocity Speed=&RandomRangeFloat(Minimum=0, Maximum=3400)
   )
  ]
  particleUpdateScript [
   &ParticleState()
   &VortexForce(Origin Pull Amount=200, UseOriginOffset=false, Vortex Axis=<1 0 0>, Vortex Axis Coordinate Space:ENiagaraCoordinateSpace=::Local) {module=VortexForce001}
   &Drag(Use Linear Drag=true, Use Rotational Drag=true)
   &GravityForce()
   &SolveForcesAndVelocity(Clamp Velocity=false, Limit Acceleration=false)
   &ScaleSpriteSizeBySpeed(Max Scale Factor=<1 2>, Min Scale Factor=<0.5 1>, Velocity Threshold=12000)
   &ScaleSpriteSize(Uniform Curve Sprite Scale=DataInterface() {propertyValues {Curve {
       defaultValue=0
       keys [{interpMode=RCIM_Cubic, tangentMode=RCTM_User, time=0, value=1}, {interpMode=RCIM_Cubic, tangentMode=RCTM_User, time=1, value=0}]
      }}})
  ]
  rendererClasses [&NiagaraSpriteRendererProperties]
  renderers [{rendererClass=&NiagaraSpriteRendererProperties}]
 }
 Sparks_To_Beam_End {
  simTarget=GPUComputeSim
  emitterSpawnScript [/SetVariables(Emitter.VectorToBeamEnd:Vector3f=&NormalizeVector(
     Vector To Normalize=&Subtract_Position(A=$Engine.Owner.Position, B=$User.BeamEnd_0)
     Safe Normalize=true
     Safe Fallback Threshold=0
     Safe Fallback Vector=<0 0 0.001>
    )) {isSetParameters=true}]
  emitterUpdateScript [
   &EmitterState(Life Cycle Mode=::NewEnumerator1, Loop Duration=1, UseLoopDelay=false)
   &SpawnRate(Spawn Probability=0.2, SpawnRate=100, Use Spawn Probability=true)
  ]
  particleSpawnScript [
   &InitializeParticle(
    Color=<rgba 275 65.3025 230.0222 1>
    Lifetime Min=0.5
    Lifetime Mode=::NewEnumerator1
    Position=$User.BeamEnd_0
    Position Mode=::NewEnumerator1
    Position Offset=<0 0 100>
    Sprite Size=&Multiply_Vector2D_ByFloat(Float=&RandomRangeFloat(Minimum=0.1, Maximum=1), Vector2D=<40 65>)
    Sprite Size Mode=::NewEnumerator4
    UsePositionOffset=true
    Write Material Random=true
   )
   /SetVariables(Particles.VectorToBeamEnd:Vector3f=&NormalizeVector(
     Vector To Normalize=&Subtract_Position(A=$Engine.Owner.Position, B=$User.BeamEnd_0)
     Safe Normalize=true
     Safe Fallback Threshold=0
     Safe Fallback Vector=<0 0 0.001>
    )) {isSetParameters=true}
   &ShapeLocation(MaskSpawnGroup=false, Override Seed=false, Sphere Radius=50, UseOwnerScale=true)
   &AddVelocity(
    Cone Angle=65
    Cone Axis=$Particles.VectorToBeamEnd
    Module.UseConeFalloff=false
    Module.UseVelDistribution=false
    Override Seed=false
    Rotation Coordinate Space=::World
    Velocity Mode=::NewEnumerator2
    Velocity Speed=&RandomRangeFloat(Minimum=200, Maximum=3400)
   )
  ]
  particleUpdateScript [
   &ParticleState()
   &VortexForce(
    Origin Pull Amount=50
    UseOriginOffset=false
    Vortex Axis=$Emitter.VectorToBeamEnd
    Vortex Axis Coordinate Space:ENiagaraCoordinateSpace=::World
   )
   &Drag(Use Linear Drag=true, Use Rotational Drag=true)
   &GravityForce()
   &SolveForcesAndVelocity(Clamp Velocity=false, Limit Acceleration=false)
   &ScaleSpriteSizeBySpeed(Max Scale Factor=<1 2>, Min Scale Factor=<0.5 1>, Velocity Threshold=12000)
   &ScaleSpriteSize(Uniform Curve Sprite Scale=DataInterface() {propertyValues {Curve {
       defaultValue=0
       keys [{interpMode=RCIM_Cubic, tangentMode=RCTM_User, time=0, value=1}, {interpMode=RCIM_Cubic, tangentMode=RCTM_User, time=1, value=0}]
      }}})
  ]
  rendererClasses [&NiagaraSpriteRendererProperties]
  renderers [{rendererClass=&NiagaraSpriteRendererProperties}]
 }
}
summary {
 systemName=NS_PistolCanaryCD_Connecting_Spire_Beam_Loop
 emitters [
  {emitterName=Bright_Ribbon_01, rendererClasses [&NiagaraRibbonRendererProperties]}
  {emitterName=Faded_Ribbon_01, rendererClasses [&NiagaraRibbonRendererProperties]}
  {emitterName=Bright_Ribbon_02, rendererClasses [&NiagaraRibbonRendererProperties]}
  {emitterName=Sparks_Main, simTarget=GPUComputeSim, rendererClasses [&NiagaraSpriteRendererProperties]}
  {emitterName=Sparks_To_Beam_End, simTarget=GPUComputeSim, rendererClasses [&NiagaraSpriteRendererProperties]}
  {emitterName=Faded_Ribbon_02, rendererClasses [&NiagaraRibbonRendererProperties]}
 ]
 userVariables [
  User.BeamEnd_0 {type=NiagaraPosition, description="", defaultValue {}}
  User.BeamFade {type=NiagaraFloat, defaultValue=1, description=""}
  User.bHitReaction {type=NiagaraBool}
  User.DisappearTime {type=NiagaraFloat, defaultValue=0.8, description=""}
  User.Duration {type=NiagaraFloat, defaultValue=30, description=""}
  User.Initial Start Time {type=NiagaraFloat, defaultValue=0, description=""}
  User.InterpolateRateOfChange {type=NiagaraFloat, defaultValue=20, description=""}
  User.Stop Location {type=NiagaraPosition, description="", defaultValue {}}
 ]
}
